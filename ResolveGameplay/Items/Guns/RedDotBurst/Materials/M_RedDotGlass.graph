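# M_RedDotGlass is a two-sided translucent material for the red dot sight glass on the RedDotBurst gun.
# A radial-gradient vignette combines with an inverted, tiled cloud texture for smudged opacity, with a flat mobile fallback.
# Colour switches between a brightness tint and a roughness-controlled skylight reflection on deferred shading.
# A quality switch uses AttachScope's world position offset at Epic and an offset collapsing vertices to the object origin at lower levels.
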Material /ResolveGameplay/Items/Guns/RedDotBurst/Materials/M_RedDotGlass
graph {
 properties {
  bUsedWithMeshDeformer=true
  bUsedWithSkeletalMesh=true
  blendMode=BLEND_Translucent
  pixelDepthOffsetMode=PDOM_Legacy
  twoSided=true
  usedShadingModels=Default Lit
 }
 nodes [
  Constant3Vector()
  Constant() {r=1}
  FeatureLevelSwitch()
  FeatureLevelSwitch()
  LinearInterpolate()
  n1 = VectorParameter() {parameterName=Brightness, defaultValue=<rgba 0.4015 0.4244 0.4427 1>}
  LinearInterpolate(B=@n1.RGB)
  n2 = MaterialFunctionCall(
   CenterPosition=Constant2Vector() {g=0.5, r=0.5}
   Radius=ScalarParameter() {defaultValue=1.33, parameterName=Radius}
   Density=ScalarParameter() {defaultValue=1, parameterName=Refrac}
  ) {materialFunction=&RadialGradientExponential}
  n3 = Saturate(None=@n2.RadialGradientExponential)
  n4 = LinearInterpolate(
   A=ScalarParameter() {defaultValue=1.33, parameterName=Vignette Amount}
   B=ScalarParameter() {defaultValue=-0.2, parameterName=Vignette center}
   Alpha=@n3
  ) {constB=0}
  n5 = TextureSample(UVs=TextureCoordinate() {uTiling=1.5, vTiling=1.5}) {texture=/Game/Weapons/FORT_Pistols/Effects/Muzzle/T_Cloud_d}
  n6 = Saturate(None=LinearInterpolate(
    A=ScalarParameter() {defaultValue=0.05, parameterName=Min}
    B=ScalarParameter() {defaultValue=1.2, parameterName=Max}
    Alpha=OneMinus(None=@n5.G)
   ))
  Multiply(A=@n4, B=@n6)
  OneMinus(None=@n3)
  Power(Base=@n2.RadialGradientExponential) {constExponent=4}
  n7 = Multiply(A=Subtract(A=WorldPosition().XYZ, B=ObjectPositionWS())) {constB=-1}
  QualitySwitch(
   Default=@n7
   Low=@n7
   High=@n7
   Medium=@n7
   Epic=MaterialFunctionCall!(RotateDialMask, Normal WS) {materialFunction=/Game/Packages/Fortress_SharedMaterials/Base_Material_Functions/AttachScope}.World Position Offset
  )
  Saturate()
  ScalarParameter() {defaultValue=0.999, parameterName=Refract}
  SceneColor!(UVs)
  ShadingPathSwitch(
   Default=@n1.RGB
   Deferred=SkyLightEnvMapSample(Direction=PixelNormalWS(), Roughness=ScalarParameter() {defaultValue=0.2, parameterName=Roughness})
  )
  ShadingPathSwitch(Default=Saturate(None=Add(A=@n4, B=@n6)), Mobile=ScalarParameter() {defaultValue=0.2, parameterName=MobileOpacity})
  SphereMask()
  StaticBool() {value=true}
  TextureSample() {texture=/Game/Packages/Fortress_SharedMaterials/Textures/T_UV_Checker_D}
 ]
}
